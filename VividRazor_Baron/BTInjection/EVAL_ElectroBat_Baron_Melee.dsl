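; Behavior tree evaluator that decides whether the ElectroBat Baron boss's melee behavior may be used, judging by its file name.
; On enter, it caches the AI controller's controlled pawn.
; Starting conditions require the parent check to pass and the pawn to lack both the ShieldsBroken tag and any PreventUsage tag.
; Ongoing conditions call the parent but always return true, so those tags only gate the start.
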(event EventOnEnter
  (params (OwnerComp :type Behavior%20Tree%20Component%20Object%20Reference))
  (bind _ascontroller (Utilities|Casting|CastToController (Components|GetOwner OwnerComp)))
  (Variables|Default|SetPawn (Pawn|GetControlledPawn _ascontroller)))

(fn EvaluateOngoingConditions
  (params (OwnerComp :type Behavior%20Tree%20Component%20Object%20Reference))
  (|CallParentFunction OwnerComp)
  (return true))

(fn EvaluateStartingConditions
  (params (OwnerComp :type Behavior%20Tree%20Component%20Object%20Reference))
  (bind _pawn (Variables|Default|GetPawn))
  (bind _returnvalue (|CallParentFunction OwnerComp))
  (if _returnvalue
    (bind _returnvalue_2 (GameplayTags|HasMatchingGameplayTag(Message) _pawn "(TagName=\"Ability.VividRazor.Boss.Baron.ShieldsBroken\")"))
    (if (not _returnvalue_2)
      (bind _returnvalue_3 (GameplayTags|HasAnyMatchingGameplayTags(Message) _pawn (Variables|Default|GetTags_PreventUsage)))
      (return (not _returnvalue_3))
      (else
        (return false)))
    (else
      (return false))))
