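; Behavior tree injection evaluator that gates the VividRazor Baron boss's TripleClaw, judging by its name an attack, on the boss's shield state.
; On enter, it resolves the owning controller's pawn and caches it.
; The starting-condition check passes only if the parent evaluation succeeds and that cached pawn has the Ability.VividRazor.Boss.Baron.ShieldsBroken gameplay tag.
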(event EventOnEnter
  (params (OwnerComp :type Behavior%20Tree%20Component%20Object%20Reference))
  (bind _ascontroller (Utilities|Casting|CastToController (Components|GetOwner OwnerComp)))
  (Variables|Default|SetPawn (Pawn|GetControlledPawn _ascontroller)))

(fn EvaluateStartingConditions
  (params (OwnerComp :type Behavior%20Tree%20Component%20Object%20Reference))
  (bind _returnvalue (|CallParentFunction OwnerComp))
  (if _returnvalue
    (bind _returnvalue_2 (GameplayTags|HasMatchingGameplayTag(Message) (Variables|Default|GetPawn) "(TagName=\"Ability.VividRazor.Boss.Baron.ShieldsBroken\")"))
    (return _returnvalue_2)
    (else
      (return false))))
